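# Niagara environment effect for desert dust: occasional tan dust puffs and bursts of small wind-blown bits, both drifting mainly along +X.
# Particle_Emitter spawns long-lived sprites that grow, rotate and fade, with curves animating dynamic material parameters.
# BitsGPU runs on GPU with curve-pulsed spawning, speed-based stretching and a tiled turbulence vector field applied as velocity.
# Both loop and spawn in flattened spheres whose local offset follows a curve over emitter age.
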NiagaraSystem /Game/Effects/Fort_Effects/Effects/Environment/Desert/Dust/NS_Desert_Dust
dependencies {
 usedDataInterfaces [
  &NiagaraDataInterfaceVectorCurve
  &NiagaraDataInterfaceCurve
  &NiagaraDataInterfaceSpriteRendererInfo
  &NiagaraDataInterfaceVector2DCurve
  /Script/Niagara.NiagaraDataInterfaceVectorField
 ]
 usedDynamicInputs []
 usedModules [&SystemState]
}
emitters {
 BitsGPU {
  simTarget=GPUComputeSim
  emitterUpdateScript [
   &EmitterState(Life Cycle Mode=::NewEnumerator1, Loop Duration=1, UseLoopDelay=true)
   &SpawnRate(
    SpawnRate=&FloatFromCurve(FloatCurve=DataInterface() {propertyValues {Curve {keys [<at 0 0>,<at 0.25 60>,<at 1 0>]}}}, CurveIndex=$Emitter.NormalizedLoopAge)
    Use Spawn Probability=false
   )
  ]
  particleSpawnScript [
   /SetVariables(Particles.ShapeLocationOffset:Vector3f=&VectorFromCurve(
     VectorCurve=DataInterface() {propertyValues {
       XCurve {keys [<at 0 0>,<at 0.3 325>,<at 0.7 -180>,<at 1 450>]}
       YCurve {keys [<at 0 0>,<at 0.3 0>,<at 0.7 0>,<at 1 0>]}
       ZCurve {keys [<at 0 12>,<at 0.3 130>,<at 0.7 12>,<at 1 320>]}
      }}
     CurveIndex=$Emitter.LoopedAge
    )) {isSetParameters=true}
   /SetVariables(Particles.ShapeLocationOffsetSpace:ENiagaraCoordinateSpace=::Local) {isSetParameters=true}
   &InitializeParticle(
    Lifetime=&RandomRangeFloat(Minimum=1, Maximum=1.5, Evaluation Type=::NewEnumerator1)
    Sprite Size=&Vector2DFromFloat(Value=&RandomRangeFloat(Minimum=2, Maximum=4, Evaluation Type=::NewEnumerator1))
    Sprite Size Mode=::NewEnumerator4
    UsePositionOffset=false
    Write Material Random=true
   )
   &Color(Color=&MakeLinearColorFromVectorAndFloat("Vector (RGB)"=<0.2 0.2 0.2>, "Float (Alpha)"=0.5))
   &AddVelocity(
    Rotation Coordinate Space=::World
    Velocity=&RandomRangeVector(Minimum=<12 0 0>, Maximum=<625 0 0>, Evaluation Type=::NewEnumerator1)
   )
   &ShapeLocation(
    MaskSpawnGroup=false
    Non Uniform Scale=<1 1 0>
    Offset=$Particles.ShapeLocationOffset
    Offset Coordinate Space=$Particles.ShapeLocationOffsetSpace
    Offset Mode=::NewEnumerator0
    Override Seed=false
    Sphere Radius=&RandomRangeFloat(Minimum=250, Maximum=400, Evaluation Type=::NewEnumerator1)
    UseOwnerScale=true
   )
  ]
  particleUpdateScript [
   &ParticleState()
   &ScaleSpriteSize(
    Scale Factor=&Vector2DFromCurve(
     Vector2Curve=DataInterface() {propertyValues {XCurve {keys [<at 0.75 1 RCIM_Cubic>,<at 1 0 RCIM_Cubic>]}, YCurve {keys [<at 0.75 1 RCIM_Cubic>,<at 1 0 RCIM_Cubic>]}}}
     ScaleCurve=<1 1>
    )
    Scale Sprite Size Mode=::NewEnumerator2
   )
   &SubUVAnimation(
    "Sprite Renderer "=DataInterface() {propertyValues {SpriteRenderer=&self:BitsGPU_0.NiagaraSpriteRendererProperties_2}}
    UseEndFrame=false
    UseStartFrame=false
   )
   &Drag(Drag=0.5, Use Linear Drag=true, Use Rotational Drag=true)
   &AccelerationForce(Acceleration=<30 0 30>)
   &SolveForcesAndVelocity(Clamp Velocity=false, Limit Acceleration=false)
   &ScaleSpriteSizeBySpeed(Max Scale Factor=<7 20>, Source Velocity:Vector3f=&Multiply_Vector(A=<0.05 0.05 1>, B=$Particles.Velocity))
   &ScaleColor(ScaleA=true, ScaleRGB=true)
   /Niagara/VectorFields/SampleVectorField(
    FieldScale:Vector3f=<15 15 15>
    VectorField=DataInterface() {propertyValues {Field=/Game/Effects/VectorFields/Noisy_Turbulence_103, bTileX=true, bTileY=true, bTileZ=true}}
   )
   /Niagara/VectorFields/ApplyVectorField(
    Vector Field Intensity as Force=0
    Vector Field Intensity as Velocity:NiagaraFloat=&FloatFromCurve(FloatCurve=DataInterface() {propertyValues {Curve {defaultValue=1, keys [{tangentMode=RCTM_User, time=0, value=15}, {tangentMode=RCTM_User, time=0.5, value=1}]}}})
    Vector Field Velocity Tightness=0.1
   )
  ]
  rendererClasses [&NiagaraSpriteRendererProperties]
  renderers [{rendererClass=&NiagaraSpriteRendererProperties}]
 }
 Particle_Emitter {
  emitterUpdateScript [
   &EmitterState(Life Cycle Mode=::NewEnumerator1, Loop Duration=1, UseLoopDelay=true)
   &SpawnRate(
    SpawnRate=&RandomRangeFloat(Minimum=0, Maximum=1.25, Evaluation Type=::NewEnumerator1)
    Use Spawn Probability=false
   )
  ]
  particleSpawnScript [
   /SetVariables(Particles.SpriteRotationRate=&RandomRangeFloat(Minimum=0.03, Maximum=0.05, Evaluation Type=::NewEnumerator1)) {isSetParameters=true}
   /SetVariables(Particles.Acceleration:Vector3f=&RandomRangeVector(Minimum=<30 0 -8>, Maximum=<15 0 -5>, Evaluation Type=::NewEnumerator1)) {isSetParameters=true}
   /SetVariables(Particles.ShapeLocationOffset:Vector3f=&VectorFromCurve(
     VectorCurve=DataInterface() {propertyValues {
       XCurve {keys [<at 0 0>,<at 0.3 325>,<at 0.7 -180>,<at 1 450>]}
       YCurve {keys [<at 0 0>,<at 0.3 0>,<at 0.7 0>,<at 1 0>]}
       ZCurve {keys [<at 0 0>,<at 0.3 0>,<at 0.7 0>,<at 1 0>]}
      }}
     CurveIndex=$Emitter.LoopedAge
    )) {isSetParameters=true}
   /SetVariables(Particles.ShapeLocationOffsetSpace:ENiagaraCoordinateSpace=::Local) {isSetParameters=true}
   &InitializeParticle(
    Color=<rgba 0.7656 0.5939 0.4267 1>
    Lifetime=&RandomRangeFloat(Minimum=6, Maximum=8, Evaluation Type=::NewEnumerator1)
    Sprite Rotation Angle=&RandomRangeFloat(Minimum=0, Maximum=1, Evaluation Type=::NewEnumerator1)
    Sprite Rotation Mode=::NewEnumerator1
    Sprite Size=&Vector2DFromFloat(Value=&RandomRangeFloat(Minimum=600, Maximum=250, Evaluation Type=::NewEnumerator1))
    Sprite Size Mode=::NewEnumerator4
    UsePositionOffset=false
    Write Material Random=true
   )
   &AddVelocity(Velocity=&RandomRangeVector(Minimum=<180 -180 80>, Maximum=<450 180 60>, Evaluation Type=::NewEnumerator1))
   &Color(Color=&MakeLinearColorFromVectorAndFloat(
     "Vector (RGB)"=<0.7656 0.5939 0.4267>
     "Float (Alpha)"=&RandomRangeFloat(Minimum=0.9, Maximum=1.2, Evaluation Type=::NewEnumerator1)
    )) {enabled=false}
   &ShapeLocation(
    MaskSpawnGroup=false
    Non Uniform Scale=<1 1 0>
    Offset=$Particles.ShapeLocationOffset
    Offset Coordinate Space=$Particles.ShapeLocationOffsetSpace
    Offset Mode=::NewEnumerator0
    Override Seed=false
    Sphere Radius=&RandomRangeFloat(Minimum=150, Maximum=300, Evaluation Type=::NewEnumerator1)
    UseOwnerScale=true
   )
  ]
  particleUpdateScript [
   &ParticleState()
   &ScaleSpriteSize(
    Scale Sprite Size Mode=::NewEnumerator0
    Uniform Scale Factor=&FloatFromCurve(FloatCurve=DataInterface() {propertyValues {Curve {keys [<at 0 0.086 RCIM_Cubic>, {arriveTangent=2.914, interpMode=RCIM_Cubic, leaveTangent=2.914, time=0.1592, value=0.9267}, <at 1 3 RCIM_Cubic>]}}})
   )
   &SpriteRotationRate(Rotation Rate=&AngleConversion(
     Angle:NiagaraFloat=$Particles.SpriteRotationRate
     Angle Input:ENiagara_AngleInput=::NewEnumerator1
     Angle Output:ENiagara_AngleInput=::NewEnumerator0
    ))
   &AccelerationForce(Acceleration=$Particles.Acceleration)
   &Drag(Drag=0.55, Use Linear Drag=true, Use Rotational Drag=true)
   &SolveForcesAndVelocity(Clamp Velocity=false, Limit Acceleration=false)
   &ScaleColor(
    Scale Alpha=&FloatFromCurve(FloatCurve=DataInterface() {propertyValues {Curve {
        defaultValue=1
        keys [
         {interpMode=RCIM_Cubic, tangentMode=RCTM_User, time=0, value=0}
         {interpMode=RCIM_Cubic, tangentMode=RCTM_User, time=0.5, value=1}
         {tangentMode=RCTM_User, time=1, value=0}
        ]
       }}})
    ScaleA=true
    ScaleRGB=true
   )
   &DynamicMaterialParameters(
    Index 0 Param 1=&FloatFromCurve(FloatCurve=DataInterface() {propertyValues {Curve {keys [
         <at 0.6907 1 RCIM_Cubic>
         {arriveTangent=73.3687, interpMode=RCIM_Cubic, leaveTangent=73.3687, tangentMode=RCTM_User, time=0.8356, value=6.6246}
         <at 0.9809 21.0252 RCIM_Cubic>
        ]}}})
    Index 0 Param 2=&FloatFromCurve(FloatCurve=DataInterface() {propertyValues {Curve {keys [<at 0 0.025 RCIM_Cubic>,<at 0.8 0.05 RCIM_Cubic>]}}})
    Index 0 Param 3=&FloatFromCurve(FloatCurve=DataInterface() {propertyValues {Curve {keys [
         <at 0.2663 0.8858 RCIM_Cubic>
         {arriveTangent=-1.4862, interpMode=RCIM_Cubic, leaveTangent=-1.4862, time=0.416, value=0.8196}
         <at 0.6465 0.3207 RCIM_Cubic>
        ]}}})
    Index 0 Param 4=1
    Module.Param0WriteEnabled=true
    Module.Param1WriteEnabled=true
    Module.Param2WriteEnabled=true
    Module.Param3WriteEnabled=true
   )
   &SubUVAnimation(
    End Frame Range Override=2
    Evaluation Type=::NewEnumerator0
    "Sprite Renderer "=DataInterface() {propertyValues {SpriteRenderer=&self:Particle_Emitter_0.NiagaraSpriteRendererProperties_1}}
    SubUV Animation Mode:ENiagara_SubUVLookupModeV2=::NewEnumerator2
    UseEndFrame=true
    UseStartFrame=true
   )
  ]
  rendererClasses [&NiagaraSpriteRendererProperties]
  renderers [{rendererClass=&NiagaraSpriteRendererProperties}]
 }
}
summary {
 systemName=NS_Desert_Dust
 emitters [
  {emitterName=Particle_Emitter, rendererClasses [&NiagaraSpriteRendererProperties]}
  {emitterName=BitsGPU, simTarget=GPUComputeSim, rendererClasses [&NiagaraSpriteRendererProperties]}
 ]
 userVariables []
}
